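# Vented wheel material that adds an animated, tinted emissive glow from a rim mask texture on top of the shared MF_WheelParent_Simple material attributes.
# Panning aurora noise distorts scrolling cloud noise via bump offset, which is contrast-adjusted, colored and masked by the rim texture.
# The glow blends into Emissive by inverted red vertex color, then passes through vehicle damage and de-rez transition functions.
# A separate pulsing tire-emissive branch is present but feeds nothing.
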Material /VehicleCosmetics/Wheels/Vented/Materials/M_Wheel_Vented
graph {
 properties {
  bAutomaticallySetUsageInEditor=false
  bUseMaterialAttributes=true
  bUsedWithMeshDeformer=true
  bUsedWithSkeletalMesh=true
  pixelDepthOffsetMode=PDOM_Legacy
  usedShadingModels=Default Lit
 }
 nodes [
  Constant()
  n1 = TextureCoordinate() {uTiling=2}
  n2 = Time()
  n3 = TextureSample(UVs=Rotator(Coordinate=Panner(Coordinate=@n1, Time=@n2) {speedX=0.5}, Time=Constant() {r=1.55}) {centerX=0, centerY=0, speed=1}) {samplerSource=SSM_Wrap_WorldGroupSettings, texture=/VehicleCosmetics/Wheels/_SharedTextures/FX/Universal_Noise_03_RGB}
  n4 = Clamp(None=Cosine(None=Multiply(A=@n2) {constB=0.25}) {period=3}) {minDefault=0.2}
  n5 = TextureObjectParameter() {group=5_Special Effects, parameterName=Emissive Masks, samplerType=SAMPLERTYPE_LinearColor, sortPriority=0, texture=null}
  n6 = TextureSample(Tex=@n5) {samplerSource=SSM_Wrap_WorldGroupSettings, samplerType=SAMPLERTYPE_LinearColor}
  n7 = Multiply(A=@n4, B=@n6.R) {constB=4}
  n8 = Multiply(A=@n6.G) {constB=4}
  n9 = Add(
   A=Multiply(A=Multiply(A=@n3.B, B=@n4), B=@n6.B)
   B=Add(A=Add(A=@n7, B=@n8), B=TextureSample(Tex=@n5) {samplerSource=SSM_Wrap_WorldGroupSettings, samplerType=SAMPLERTYPE_LinearColor}.A) {constB=0.2}
  )
  n10 = ScalarParameter() {defaultValue=3, group=5_Special Effects, parameterName=Tire Emissive Brightness, sortPriority=0}
  n11 = VectorParameter() {group=5_Special Effects, parameterName=Emissive Color, sortPriority=0, defaultValue=<rgba 0 0.2307 1 1>}
  n12 = VectorParameter() {group=5_Special Effects, parameterName=Emissive Color 02, sortPriority=0, defaultValue=<rgba 1 0.9324 0 1>}
  LinearInterpolate(
   A=Multiply(A=Multiply(A=@n9, B=@n10), B=@n11.RGB)
   B=@n12.RGB
   Alpha=Multiply(
    A=TextureSampleParameter2D() {parameterName=Emissive 02, texture=null}.A
    B=ScalarParameter() {defaultValue=3, group=5_Special Effects, parameterName=Tire Emissive Brightness 02, sortPriority=0}
   )
  )
  n13 = TextureSampleParameter2D() {
   group=5_Special Effects
   parameterName=Effect_E
   samplerType=SAMPLERTYPE_LinearColor
   sortPriority=1
   texture=/VehicleCosmetics/Wheels/Vented/Textures/T_Vented_Rim_E
  }
  n14 = VectorParameter() {group=5_Special Effects, parameterName=Custom Color, sortPriority=1, defaultValue=<rgba 1 0.2575 0 1>}
  n15 = Multiply(A=@n14.RGB, B=ScalarParameter() {defaultValue=16, group=5_Special Effects, parameterName=Emissive Brightness, sortPriority=1})
  n16 = TextureCoordinate() {uTiling=2, vTiling=2}
  n17 = TextureSample(UVs=Panner(Coordinate=@n16) {speedX=-0.25, speedY=0.1}) {samplerSource=SSM_Wrap_WorldGroupSettings, texture=/VehicleCosmetics/Wheels/_SharedTextures/FX/aurora_T}
  n18 = Add(A=Panner(Coordinate=@n16) {speedX=-1, speedY=0.1}, B=@n17.G)
  n19 = Clamp(None=TextureSample(UVs=BumpOffset(Coordinate=@n18, Height=@n17.G) {heightRatio=1}) {samplerSource=SSM_Wrap_WorldGroupSettings, texture=/VehicleCosmetics/Wheels/_SharedTextures/FX/Cloud_04}.G) {minDefault=0.1}
  n20 = ScalarParameter() {defaultValue=1, group=5_Special Effects, parameterName=Contrast}
  n21 = Multiply(
   A=Multiply(
    A=Multiply(A=@n13.R, B=@n15)
    B=MaterialFunctionCall(In=@n19, Contrast=@n20) {materialFunction=&CheapContrast}.Result
   )
   B=Power(Base=@n13.G, Exp=ScalarParameter() {defaultValue=1, group=5_Special Effects, parameterName=AO Power, sortPriority=1})
  )
  LinearInterpolate(B=@n21, Alpha=OneMinus(None=VertexColor().G))
  n22 = MaterialFunctionCall() {materialFunction=&MF_WheelParent_Simple}
  n23 = GetMaterialAttributes(None=@n22.MaterialAttributes) {attributeGetTypes [#1]}
  n24 = LinearInterpolate(A=@n23.Emissive Color, B=NamedRerouteUsage(), Alpha=OneMinus(None=@n22.VertexColorR))
  MaterialFunctionCall(In=MaterialFunctionCall(Material=SetMaterialAttributes(MaterialAttributes=@n23.MaterialAttributes, Emissive Color=@n24) {attributeSetTypes [#1]}) {materialFunction=&MF_BaseVehicle_Damage}.Result) {materialFunction=&MF_VehicleDeRez_Transition}
  NamedRerouteDeclaration(None=@n21) {name=Emissive, nodeColor=<rgba 1 0.75 0 1>}
 ]
}
